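; Creative atmospheric particles device whose enabled state follows an Enabled On Phase user option.
; Loading or changing options re-evaluates that state as PreGame; the effect itself is not defined here.
; Bounding box calculation excludes primitive components under the additional visible item root.
; Legacy migration converts a non-default EnabledIndex to the phase enum, offset by one, then resets it and refreshes property overrides.
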(event OnPropertiesLoaded(ToyOptionsComponent)
  (CallFunction|UpdateEnabledfromPhase "PreGame"))

(event OnAnyPropertyChanged(ToyOptionsComponent)
  (CallFunction|UpdateEnabledfromPhase "PreGame"))

(fn BP_GetComponentsBoundingBox
  (params (bNonColliding :type Boolean) (bIncludeFromChildActors :type Boolean))
  (Utilities|Array|Clear (LocalVariables|GetPrimitiveComponentsToIgnore))
  (for _array_element (Components|GetChildrenComponents (Variables|Default|GetAdditionalVisibleItemRoot) true)
    (bind _asprimitive_component (Utilities|Casting|CastToPrimitiveComponent _array_element))
    (Utilities|Array|Add (LocalVariables|GetPrimitiveComponentsToIgnore) _asprimitive_component))
  (return (CallFunction|GetBoundingBoxwithIgnoredComponents (LocalVariables|GetPrimitiveComponentsToIgnore) bNonColliding bIncludeFromChildActors)))

(fn GetEnabledPhaseasInteger
  ;[0 0.6223 1 0.25] (Legacy) Enabled Index is offset by 1 from the enum (includes never as -1 instead of 0)
  (return (- (Math|Conversions|ToInteger(Byte) (Variables|UserOptions|Basic|GetEnabled%20On%20Phase)) 1)))

(fn MigrateLegacyUserOptions
  (bind (_enabled_on_phase _enabledindex) (Class|GetClassDefaults "/Game/Creative/Items/Atmospheric_Particles/CP_P_Atmospheric_Aesthetic.CP_P_Atmospheric_Aesthetic_C"))
  (bind _enabledindex_2 (Variables|UserKnobs|GetEnabledIndex))
  ;[1 0 0.8267 0.25] Migrate old EnabledIndex into new enum (offset by 1)
  (|CallParentFunction)
  (if (!= _enabledindex_2 _enabledindex)
    (Variables|UserOptions|Basic|SetEnabled%20On%20Phase (Utilities|Enum|BytetoEnumEEnabledOnPhase (Math|Conversions|ToByte(Integer) (+ (Variables|UserKnobs|GetEnabledIndex) 1))))
    (Variables|SetBy-RefVar _enabledindex_2 _enabledindex)
    (OptionData|RefreshPropertyOverrides (Variables|Default|GetToyOptionsComponent))))

(fn ConstructionScript
  (|CallParentFunction))
