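; Fortnite Creative atmospheric particle prop that shows an aesthetic Niagara effect, with enablement set by a game-phase user option.
; SpawnParticles async-loads the system, preferring a valid mobile variant on mobile, then applies vector and scalar parameters and spawns audio.
; Option changes re-evaluate bEnabled on authority using PreGame; BeaconUpdate deactivates Niagara cosmetically when disabled.
; The construction script clears the inherited Cascade template in editor to prevent misuse.
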(event Custom|Load%20and%20Spawn%20NS%20Particle
  (params (NiagaraSystemTemplate :type Niagara%20System%20Soft%20Object%20Reference))
  (bind _loaded_object (Creative|Loading|AsyncLoadEditorSafe NiagaraSystemTemplate)
    (:Loaded
      (bind _asniagara_system (Utilities|Casting|CastToNiagaraSystem _loaded_object))
      (Components|Activation|Activate (Variables|Default|GetNiagara) false)
      (Niagara|SetNiagaraSystemAsset (Variables|Default|GetNiagara) _asniagara_system true)
      (Effects|Components|ParticleSystem|SetAutoAttachmentParameters (Variables|Default|GetNiagara) (Variables|Transformation|GetRootComponent) "None" "KeepRelative" "KeepRelative" (select (Variables|Default|GetParticlePropAbsoluteScale) "KeepWorld" "KeepRelative"))
      (EventDispatchers|BindEventtoOnSystemFinished (Variables|Default|GetNiagara) (EventDispatchers|CreateEvent :!event Custom|On%20NS%20Finished))
      (CallFunction|SpawnParticleAudio "None")
      (Variables|/Script/Engine.SceneComponent|SetbAbsoluteScale (Variables|Default|GetParticlePropAbsoluteScale) (Variables|Default|GetNiagara))
      (bind _values (Utilities|Map|Values (Variables|Default|GetParticleSystemVectorParameters)))
      (bind _keys (Utilities|Map|Keys (Variables|Default|GetParticleSystemVectorParameters)))
      (for (_array_element _array_index) _keys
        (Effects|Components|ParticleSystem|SetVectorParameter (Variables|Default|GetNiagara) _array_element (Math|Conversions|ToVector(LinearColor) (Utilities|Array|Get(acopy) _values _array_index))))
      (bind _values_2 (Utilities|Map|Values (Variables|Default|GetParticleSystemScalarParameters)))
      (bind _keys_2 (Utilities|Map|Keys (Variables|Default|GetParticleSystemScalarParameters)))
      (for (_array_element_2 _array_index_2) _keys_2
        (Effects|Components|ParticleSystem|SetFloatParameter (Variables|Default|GetNiagara) _array_element_2 (Utilities|Array|Get(acopy) _values_2 _array_index_2))))))

(event Custom|On%20NS%20Finished
  (params (PSystem :type Niagara%20Particle%20System%20Component%20Object%20Reference))
  (CallFunction|DespawnParticleAudio "None"))

(event OnAnyPropertyChanged(ToyOptionsComponent)
  (CallFunction|UpdateEnabledfromPhase "PreGame"))

(event OnPropertiesLoaded(ToyOptionsComponent)
  (CallFunction|UpdateEnabledfromPhase "PreGame"))

(fn BeaconUpdate
  (|CallParentFunction)
  (Networking|CanExecuteCosmeticEvents
    (:True
      (if (not (Variables|Default|GetbEnabled))
        (Components|Activation|Deactivate (Variables|Default|GetNiagara))))))

(fn SpawnParticles
  (Networking|CanExecuteCosmeticEvents
    (:True
      ;[1 1 1] If mobile NS asset is not specified, use default NS asset instead
      (bind _output_get (LocalVariables|SetNiagaraSystemToLoad (select (Fortnite|IsMobilePlatform) (select (Utilities|IsValidSoftObjectReference (Variables|Default|GetNiagaraSystemTemplateMobile)) (Variables|Default|GetNiagaraSystemTemplateMobile) (Variables|Default|GetNiagaraSystemTemplate)) (Variables|Default|GetNiagaraSystemTemplate))))
      (if (Utilities|IsValidSoftObjectReference _output_get)
        (CallFunction|LoadAndSpawnNSParticle _output_get)))))

(fn UpdateEnabledFromPhase
  (params (local%20Phase :type EFortMinigameState%20Enum :default "PreGame"))
  (bind _localenabled (LocalVariables|GetLocalEnabled))
  (bind _benabled (Variables|Default|GetbEnabled))
  (bind _returnvalue (Math|Boolean|NotEqual(Boolean) _benabled _localenabled))
  (LocalVariables|SetLocalGameplayPhase local%20Phase)
  (LocalVariables|SetLocalEnabled (Math|Boolean|ORBoolean (Utilities|Enum|Equal(Enum) (Variables|UserOptions|GetEnabledDuringPhase) "NewEnumerator0") (and (Utilities|Enum|Equal(Enum) (Variables|UserOptions|GetEnabledDuringPhase) "NewEnumerator2") (or (Utilities|Enum|Equal(Enum) (LocalVariables|GetLocalGameplayPhase) "PreGame") (Utilities|Enum|Equal(Enum) (LocalVariables|GetLocalGameplayPhase) "Warmup"))) (and (Utilities|Array|ContainsItem (Variables|Default|GetGameplay%20Phases) (LocalVariables|GetLocalGameplayPhase)) (Utilities|Enum|Equal(Enum) (Variables|UserOptions|GetEnabledDuringPhase) "NewEnumerator4"))))
  (Networking|SwitchHasAuthority
    (:Authority
      (if _returnvalue
        (Variables|Default|SetbEnabled (LocalVariables|GetLocalEnabled))))
    (:Remote
      (if (not _returnvalue)
        (CallFunction|BeaconUpdate)))))

(fn ConstructionScript
  (if (Editor|InEditor)
    (Variables|Default|SetParticlePropTemplate 0) ; Clear Cascade Partcile incase misused here
  ))
