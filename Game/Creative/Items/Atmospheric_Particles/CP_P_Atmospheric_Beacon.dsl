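; Creative Beacon device that shows a particle beacon and/or floating HUD marker to local players, filtered by team and class.
; Beacon Update runs only where cosmetics execute, with bindings to local team and class changes.
; ClassValid, which always passes in pregame, and TeamValid gate a style switch between beacon and value-bar HUD.
; Hiding destroys the particle component; showing respawns it and sets an optional identifier letter through emitter parameters.
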(event OnAddedtoMinigame(FortMinigameProgress)
  (params (Minigame :type Fort%20Minigame%20Object%20Reference))
  ;[0 1 0.1154] Force Update when changing class
  (Networking|CanExecuteCosmeticEvents
    (:True
      (Utilities|IsValid Minigame
        (:Is%20Valid
          (Minigame|BindEventtoOnAnyLocalPlayerChangedClass Minigame (EventDispatchers|CreateEvent :!event Beacon%20Update)))))))

(event EventBeginPlay
  ;[0 1 0.1154] Force Updating State when the player changes teams
  (|CallParentFunction)
  (Networking|CanExecuteCosmeticEvents
    (:True
      (bind _asfort_player_controller_athena (Utilities|Casting|CastToFortPlayerControllerAthena (Game|GetPlayerController 0)))
      (FortPlayerControllerAthena|BindEventtoOnLocalPlayerChangedTeams _asfort_player_controller_athena (EventDispatchers|CreateEvent :!event Beacon%20Update)))))

(event OnPropertiesLoaded(ToyOptionsComponent)
  ;[0 1 0.1154] Setting here as parent has no component options (these are defined on the children)
  (CallFunction|UpdateEnabledfromPhase "PreGame")
  (Initialization|InitializeComponentWithAffects (Variables|Default|GetCreative_UseRestriction_Component) (Variables|Default|GetFortMinigameProgress) 0 0 996 (Variables|UserKnob|GetClassIndex) 0 1 false (== (Variables|UserKnob|GetKnob_AffectsClass) 0) false))

(event OnAnyPropertyChanged(ToyOptionsComponent)
  ;[0 1 0.1154] Setting here as parent has no component options (these are defined on the children)
  (CallFunction|UpdateEnabledfromPhase "PreGame")
  (Initialization|InitializeComponentWithAffects (Variables|Default|GetCreative_UseRestriction_Component) (Variables|Default|GetFortMinigameProgress) 0 0 996 (Variables|UserKnob|GetClassIndex) 0 1 false (== (Variables|UserKnob|GetKnob_AffectsClass) 0) false))

(event EventBeaconUpdate
  ;[1 0.7982 0] Override for Beacon
  (Networking|CanExecuteCosmeticEvents
    (:True
      (if (and (Variables|Default|GetbEnabled) (Default|ClassValid))
        (switch int (Variables|UserKnob|GetStyleIndex)
          (:0
            (Class|CreativeValueBarComponent|HideValueBar (Variables|Default|GetFloatingHUDBar)))
          (:1
            (CallFunction|ShowHUD)
            (CallFunction|HideBeacon))
          (:2
            (CallFunction|ShowHUD)))
        (if (Default|TeamValid)
          (CallFunction|ShowBeacon)
          (CallFunction|SetBeaconIdentifierParmeters)
          (else
            (CallFunction|HideBeacon)))
        (else
          (Class|CreativeValueBarComponent|HideValueBar (Variables|Default|GetFloatingHUDBar))
          (CallFunction|HideBeacon))))))

(fn ActivateFromOtherDevice
  (params (Instigator :type Controller%20Object%20Reference) (Requester :type ERequesterType%20Enum :default "NewEnumerator0"))
  (return false))

(fn BP_GetComponentsBoundingBox
  (params (bNonColliding :type Boolean) (bIncludeFromChildActors :type Boolean))
  (return (CallFunction|GetBoundingBoxwithIgnoredComponents (Utilities|Array|MakeArray (Variables|Default|GetParticlePropPS)) bNonColliding bIncludeFromChildActors)))

(fn CanInteract
  (params (InteractingPawn :type Fort%20Pawn%20Object%20Reference) (InteractionBeingAttempted :type EInteractionBeingAttempted%20Enum :default "FirstInteraction") (InteractionType :type TInteractionType%20Enum :default "IT_NoInteraction"))
  (return false))

(fn ClassValid
  (bind _isvalidcontroller (LocalVariables|GetisValidController))
  (bind _valid_controller (DEPRECATED|ActorHasValidController (Variables|Default|GetCreative_UseRestriction_Component) (Fortnite|GetFirstLocalFortPlayerController true) true true))
  (LocalVariables|SetisValidController _valid_controller)
  (bind _currentminigame (Variables|/Script/FortniteGame.FortMinigameLogicComponent|GetCurrentMinigame (Variables|Default|GetFortMinigameProgress))
    (:then
      (return (or (Minigame|IsPregame _currentminigame) _isvalidcontroller)))
    (:else
      (return _isvalidcontroller))))

(fn GetDeviceColor
  (bind (_value _returnvalue) (Utilities|Map|Find (Variables|Default|GetParticleSystemVectorParameters) "Color"))
  (bind (_value_2 _returnvalue_2) (Utilities|Map|Find (Variables|Default|GetParticleSystemVectorParameters) "Light"))
  (return (select (Utilities|Array|ContainsItem (Variables|/Script/CascadeRuntime.ParticleSystemComponent|GetInstanceParameters (Variables|Default|GetParticlePropPS)) (Utilities|Struct|MakeParticleSysParam "Color" "PSPT_Color")) (Math|Color|SelectColor _value_2 "(R=0.000000,G=0.537697,B=1.000000,A=1.000000)" _returnvalue_2) (Math|Color|SelectColor _value "(R=0.000000,G=0.537697,B=1.000000,A=1.000000)" _returnvalue))))

(fn GetInstigatingController
  (return 0))

(fn GetSpawnedPickups
  (return 0))

(fn HideBeacon
  (Rendering|SetVisibility (Variables|Default|GetAdditionalVisibleItemRoot) false true)
  (bind _particlepropps (Variables|Default|GetParticlePropPS)
    (:then
      (Components|DestroyComponent _particlepropps))
    (:else)))

(fn IsEnabled
  (return false))

(fn IsHoldingItemForPawn
  (params (Item :type Item%20Variant%20Handle%20Structure) (Pawn :type Pawn%20Object%20Reference))
  (return false))

(fn SetTriggerCollision
  (params (On :type Boolean))
  (return false))

(fn SetBeaconIdentifierParmeters
  (Effects|Components|ParticleSystem|SetEmitterEnable (Variables|Default|GetParticlePropPS) "Letter" (> (Variables|Default|GetBeaconIdentifier) 0))
  (Effects|Components|ParticleSystem|SetFloatParameter (Variables|Default|GetParticlePropPS) "U_Index" (Math|Conversions|ToFloat(Integer) (- (Variables|Default|GetBeaconIdentifier) 1))))

(fn ShowBeacon
  (Rendering|SetVisibility (Variables|Default|GetAdditionalVisibleItemRoot) true true)
  (CallFunction|SpawnParticles))

(fn ShowHUD
  (if (Default|TeamValid)
    (LocalVariables|SetlocalHostileTeams (Knobs|Team|GetTeamArray self))
    (Utilities|FlowControl|Sequence
      (:then_0
        ;[1 1 1] Parse Team Data
        (if (Knobs|Team|IsTeamNeutral (Variables|UserKnob|GetFriendlyTeamIndex))
          (Utilities|Array|Clear (LocalVariables|GetlocalHostileTeams))
          (elif (Knobs|Team|IsTeamAny (Variables|UserKnob|GetFriendlyTeamIndex))
            (LocalVariables|SetlocalFriendlyTeam (LocalVariables|GetlocalHostileTeams)) ; Friendly teams take priority over hostile teams.
            (elif (Knobs|Team|IsValidTeamIndex (Variables|UserKnob|GetFriendlyTeamIndex))
              (Utilities|Array|Add (LocalVariables|GetlocalFriendlyTeam) (Knobs|Team|GetTeamValueasIndex (Variables|UserKnob|GetFriendlyTeamIndex)))))))
      (:then_1
        (Class|CreativeValueBarComponent|SpawnOrUpdateValueBar (Variables|Default|GetFloatingHUDBar) (Variables|Default|GetSelectionMesh) self "/Game/Creative/Devices/Common/Components/Creative_FloatingValueBar.Creative_FloatingValueBar_C" 0 (Variables|UserKnob|GetHealthBarDistance) 0 -1.0 8e+01 (Variables|UserKnob|GetbHealthBarRequireLOS) (LocalVariables|GetlocalFriendlyTeam) (LocalVariables|GetlocalHostileTeams) (Variables|UserKnob|GetShowTeamIndex) false (Variables|UserKnob|GetHealthBarOverrideTextFriendly) (Variables|UserKnob|GetHealthBarOverrideTextNeutral) (Variables|UserKnob|GetHealthBarOverrideTextEnemy) (Variables|UserKnob|GetHealthBarTextSize) false (Math|Conversions|ToBoolean(Integer) (Variables|UserKnob|GetDisplayDistanceText)) (Math|Conversions|ToBoolean(Integer) (Variables|UserKnob|GetClampToScreen)) (Math|Conversions|ToBoolean(Integer) (Variables|UserKnob|GetShowOffscreenArrow)) (Creative|UI|IconPicker|GetIconLibraryTexture (Variables|UserKnob|GetHealthBarIdentifierIndex)) true "(X=1.000000,Y=1.000000)" false false)))))

(fn TeamValid
  (LocalVariables|SetLocalPawnTeamIndex (Math|Conversions|ToInteger(Byte) (Team|GetActorPvPTeamIndex (Game|GetPlayerPawn 0))))
  (return (Math|Boolean|ORBoolean (Knobs|Team|IsTeamAny (Variables|UserKnob|GetShowTeamIndex)) (and (not (Knobs|Team|IsValidTeamIndex (Variables|UserKnob|GetShowTeamIndex))) (== (Variables|UserKnob|GetShowTeamIndex) (Variables|UserKnob|GetFriendlyTeamIndex))) (and (Knobs|Team|IsTeamFriendlies (Variables|UserKnob|GetShowTeamIndex)) (or (and (Knobs|Team|IsValidTeamIndex (Variables|UserKnob|GetFriendlyTeamIndex)) (== (LocalVariables|GetLocalPawnTeamIndex) (Knobs|Team|GetTeamValueasIndex (Variables|UserKnob|GetFriendlyTeamIndex)))) (Knobs|Team|IsTeamAny (Variables|UserKnob|GetFriendlyTeamIndex)))) (and (and (Knobs|Team|IsValidTeamIndex (Variables|UserKnob|GetFriendlyTeamIndex)) (!= (Knobs|Team|GetTeamValueasIndex (Variables|UserKnob|GetFriendlyTeamIndex)) (LocalVariables|GetLocalPawnTeamIndex))) (Knobs|Team|IsTeamHostiles (Variables|UserKnob|GetShowTeamIndex))) (and (Knobs|Team|IsValidTeamIndex (Variables|UserKnob|GetShowTeamIndex)) (== (Knobs|Team|GetTeamValueasIndex (Variables|UserKnob|GetShowTeamIndex)) (LocalVariables|GetLocalPawnTeamIndex))))))

(fn ConstructionScript
  (|CallParentFunction))
