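; Parent Blueprint for Creative atmospheric particle props that show an attached particle effect, toggled by minigame phase or Show/Hide receiver events.
; The replicated enabled flag drives BeaconUpdate, which cosmetically spawns or destroys the emitter and toggles component visibility.
; Spawning applies vector and scalar parameter maps and binds particle spawn/death events to play or fade audio.
; Selection mesh and idle FX appear only outside play mode, with separate editor-preview handling.
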(event EditorPreview|EventSetupForEditorPreview
  ;[0.4077 0.4497 1] [UEFN] Show the particle for Editor Presentation
  (|CallParentFunction)
  (Development|SetHiddeninGame (Variables|Default|GetSelectionMesh) false false)
  (CallFunction|SpawnParticles))

(event Custom|SpawnParticlesInPIE
  ;[1 1 1] Spawn Particles in PIE for Debug Purposes
  (CallFunction|SpawnParticles))

(event EventBeginPlay
  ;[0 1 0.1154] Initialization
  (CallFunction|BeaconUpdate)
  (if (Replay|GetIsPlayingReplay)
    (CallFunction|SelectionObjectVisibility false)))

(event OnAddedtoMinigame(FortMinigameProgress)
  (params (Minigame :type Fort%20Minigame%20Object%20Reference))
  ;[0 1 0.1154] Initialization
  (if (not (Replay|GetIsPlayingReplay))
    (Utilities|IsValid Minigame
      (:Is%20Valid
        (CallFunction|SelectionObjectVisibility (not (Minigame|IsinPlayMode Minigame :!class /Script/FortniteGame.FortMinigame)))))))

(event MinigameStateChangedDelegate(FortMinigameProgress)
  (params (Minigame :type Fort%20Minigame%20Object%20Reference) (NewMinigameState :type EFortMinigameState%20Enum :default "PreGame"))
  ;[0 0 0.25] Update on Minigame Progress
  (CallFunction|SelectionObjectVisibility (not (Minigame|IsinPlayMode Minigame :!class /Script/FortniteGame.FortMinigame)))
  (CallFunction|UpdateEnabledfromPhase NewMinigameState))

(event Custom|Spawn%20Particle%20Audio
  (params (Event%20Name :type Name) (Emitter%20Time :type Float%20%28single-precision%29) (Location :type Vector) (Velocity :type Vector))
  ;[1 0.7982 0] Particle Audio Spawn/Despawn
  (bind _atmospheric_audio (Variables|Default|GetAtmospheric%20Audio)
    (:then
      (Audio|Components|Audio|Play _atmospheric_audio 0.0))
    (:else)))

(event Custom|Show%20Particle
  (params (Instigator :type Controller%20Object%20Reference))
  ;[0.6648 0 1] Receivers
  (if (Math|Boolean|NotEqual(Boolean) (Variables|Default|GetbEnabled) true)
    (Variables|Default|SetbEnabled true)))

(event Custom|Despawn%20Particle%20Audio
  (params (EventName :type Name) (EmitterTime :type Float%20%28single-precision%29) (ParticleTime :type Integer) (Location :type Vector) (Velocity :type Vector) (Direction :type Vector))
  ;[1 0.7982 0] Particle Audio Spawn/Despawn
  (bind _atmospheric_audio (Variables|Default|GetAtmospheric%20Audio)
    (:then
      (Audio|Components|Audio|FadeOut _atmospheric_audio 0.3 0.0 "Linear" :!class /Script/Engine.AudioComponent))
    (:else)))

(event Custom|Hide%20Particle
  (params (Instigator :type Controller%20Object%20Reference))
  ;[0.6648 0 1] Receivers
  (if (Math|Boolean|NotEqual(Boolean) (Variables|Default|GetbEnabled) false)
    (Variables|Default|SetbEnabled false)))

(fn BeaconUpdate
  (Networking|CanExecuteCosmeticEvents
    (:True
      (if (Variables|Default|GetbEnabled)
        (Rendering|SetVisibility (Variables|Default|GetAdditionalVisibleItemRoot) true true)
        (Rendering|SetVisibility (Variables|Debug|GetStaticMeshComponent) true false)
        (CallFunction|SpawnParticles)
        (else
          (Rendering|SetVisibility (Variables|Default|GetAdditionalVisibleItemRoot) false true)
          (Rendering|SetVisibility (Variables|Debug|GetStaticMeshComponent) false false)
          (bind _particlepropps (Variables|Default|GetParticlePropPS)
            (:then
              (Components|DestroyComponent _particlepropps))
            (:else)))))))

(fn GetEnabledPhaseasInteger
  (return (Variables|UserKnobs|GetEnabledIndex)))

(fn GetParticleSystemTemplatetoSpawn
  (return (Variables|Default|GetParticlePropTemplate)))

(fn GetCollisionStaticMesh
  (return "0, 0, 0" (Variables|/Script/Engine.StaticMeshComponent|GetStaticMesh (Variables|Default|GetSelectionMesh))))

(fn GetMeshComponents
  (return 0))

(fn IgnoreWorldGridScaleOverride
  (return false))

(fn OnRep_Enabled
  (CallFunction|BeaconUpdate))

(fn SelectionObjectVisibility
  (params (On :type Boolean))
  (bind _selectionmesh (Variables|Default|GetSelectionMesh))
  (Development|SetHiddeninGame _selectionmesh (not On) false)
  (Components|Activation|SetActive (Variables|Default|GetIdleFX) On false)
  (Collision|SetCollisionEnabled _selectionmesh (select On "QueryOnly" "NoCollision")))

(fn SpawnParticles
  (Networking|CanExecuteCosmeticEvents
    (:True
      (bind _output_get (LocalVariables|SetParticleSystemToSpawn (CallFunction|GetParticleSystemTemplateToSpawn)))
      (Utilities|IsValid _output_get
        (:Is%20Valid
          (bind _particlepropps (Variables|Default|GetParticlePropPS)
            (:then
              (Components|DestroyComponent _particlepropps))
            (:else
              _))
          (bind _returnvalue (Effects|SpawnEmitterAttached _output_get (Variables|Transformation|GetRootComponent) "None" (Variables|Default|GetParticlePropRelativeLocation) (Variables|Default|GetParticlePropRelativeRotation)))
          (bind _output_get_2 (Variables|Default|SetParticlePropPS _returnvalue))
          (EventDispatchers|BindEventtoOnParticleSpawn _output_get_2 (EventDispatchers|CreateEvent :!event Spawn%20Particle%20Audio))
          (EventDispatchers|BindEventtoOnParticleDeath (Variables|Default|GetParticlePropPS) (EventDispatchers|CreateEvent :!event Despawn%20Particle%20Audio))
          (Variables|/Script/Engine.SceneComponent|SetbAbsoluteScale (Variables|Default|GetParticlePropAbsoluteScale) (Variables|Default|GetParticlePropPS))
          (bind _values (Utilities|Map|Values (Variables|Default|GetParticleSystemVectorParameters)))
          (bind _keys (Utilities|Map|Keys (Variables|Default|GetParticleSystemVectorParameters)))
          (for (_array_element _array_index) _keys
            (Effects|Components|ParticleSystem|SetVectorParameter (Variables|Default|GetParticlePropPS) _array_element (Math|Conversions|ToVector(LinearColor) (Utilities|Array|Get(acopy) _values _array_index))))
          (bind _values_2 (Utilities|Map|Values (Variables|Default|GetParticleSystemScalarParameters)))
          (bind _keys_2 (Utilities|Map|Keys (Variables|Default|GetParticleSystemScalarParameters)))
          (for (_array_element_2 _array_index_2) _keys_2
            (Effects|Components|ParticleSystem|SetFloatParameter (Variables|Default|GetParticlePropPS) _array_element_2 (Utilities|Array|Get(acopy) _values_2 _array_index_2))))))))

(fn UpdateEnabledFromPhase
  (params (local%20Phase :type EFortMinigameState%20Enum :default "PreGame"))
  (bind _local_enabled (LocalVariables|Getlocal%20Enabled))
  (bind _benabled (Variables|Default|GetbEnabled))
  (bind _returnvalue (Math|Boolean|NotEqual(Boolean) _benabled _local_enabled))
  (bind _enabledindex (CallFunction|GetEnabledPhaseAsInteger))
  (LocalVariables|SetLocal%20Enabled%20Index _enabledindex)
  (LocalVariables|Setlocal%20Gameplay%20Phase local%20Phase)
  (LocalVariables|Setlocal%20Enabled (Math|Boolean|ORBoolean (== (LocalVariables|GetLocal%20Enabled%20Index) 0) (and (== (LocalVariables|GetLocal%20Enabled%20Index) 1) (Utilities|Enum|Equal(Enum) (LocalVariables|Getlocal%20Gameplay%20Phase) "PreGame")) (and (Utilities|Array|ContainsItem (Variables|Default|GetGameplay%20Phases) (LocalVariables|Getlocal%20Gameplay%20Phase)) (== (LocalVariables|GetLocal%20Enabled%20Index) 2))))
  (Networking|SwitchHasAuthority
    (:Authority
      (if _returnvalue
        (Variables|Default|SetbEnabled (LocalVariables|Getlocal%20Enabled))))
    (:Remote
      (if (not _returnvalue)
        (CallFunction|BeaconUpdate)))))

(fn ConstructionScript)
