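; Creative atmospheric smoke floor prop that displays a configurable smoke plane.
; SetVisualParameters runs on construction and option changes, choosing a material by motion type and sidedness and writing colors, opacity, tiling and speed as custom primitive data.
; Construction also writes a random per-instance value, with all visual updates gated to cosmetic-capable contexts.
; MigrateLegacyUserOptions converts non-default legacy OpacityIndex values into the Opacity option.
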(event EventBeginPlay
  (|CallParentFunction)
  (Variables|Default|SetAdditionalVisibleItemRoot (Variables|Default|GetSmokePlane)))

(event Collision|EventActorBeginOverlap
  (params (OtherActor :type Actor%20Object%20Reference))
  (|CallParentFunction OtherActor) ; This node is disabled and will not be called.\nDrag off pins to build functionality.
)

(event EventTick
  (params (DeltaSeconds :type Float%20%28single-precision%29))
  (|CallParentFunction DeltaSeconds) ; This node is disabled and will not be called.\nDrag off pins to build functionality.
)

(event OnPropertiesLoaded(ToyOptionsComponent)
  (CallFunction|SetVisualParameters))

(event OnAnyPropertyChanged(ToyOptionsComponent)
  (CallFunction|SetVisualParameters))

(fn GetFogMaterial
  (return (select (Variables|UserOptions|GetDoubleSided) (Utilities|Map|Find (Variables|Default|GetMotionTypeMaterialsDoubleSided) (Variables|UserOptions|GetMotion%20Type)) (Utilities|Map|Find (Variables|Default|GetMotionTypeMaterials) (Variables|UserOptions|GetMotion%20Type)))))

(fn GetLinearColorValue
  (params (TargetArray :type Map%20of%20ECreativeColor%20Enums%20to%20Linear%20Color%20Structures :default "NewEnumerator0"))
  (bind _value (Utilities|Map|Find TargetArray (Variables|UserOptions|GetSmoke%20Color)))
  (bind (_value_r _value_g _value_b _value_a) (Math|Color|BreakColor _value))
  (return (Math|Vector4|MakeVector4 _value_r _value_g _value_b _value_a)))

(fn MigrateLegacyUserOptions
  (bind _opacityindex (Class|GetClassDefaults "/Game/Creative/Items/Atmospheric_Particles/CP_P_Atmospheric_SmokeFloor.CP_P_Atmospheric_SmokeFloor_C"))
  (bind _opacityindex_2 (Variables|Legacy|GetOpacityIndex))
  (|CallParentFunction)
  (if (!= _opacityindex _opacityindex_2)
    (Variables|UserOptions|SetOpacity (* _opacityindex_2 1e+02))
    (Variables|SetBy-RefVar _opacityindex_2 _opacityindex)
    (OptionData|RefreshPropertyOverrides (Variables|Default|GetToyOptionsComponent))))

(fn SetVisualParameters
  (bind _smoke_color (Variables|UserOptions|GetSmoke%20Color))
  (bind _colormap (Variables|Default|GetColorMap))
  (bind _value (Utilities|Map|Find _colormap _smoke_color))
  (bind (_value_color1_2_5e99fc9a4f7f4eeed0685e98fd6a26ec _value_color2_4_bb2f746b4f7bd5defd0542bf99dc459e) (Utilities|Struct|BreakColorPair _value))
  (Networking|CanExecuteCosmeticEvents
    (:True
      (Rendering|Material|SetCustomPrimitiveDataVector4 (Variables|Default|GetSmokePlane) 0 (Utilities|LinearColortoVector4 _value_color1_2_5e99fc9a4f7f4eeed0685e98fd6a26ec))
      (Rendering|Material|SetCustomPrimitiveDataVector4 (Variables|Default|GetSmokePlane) 4 (Utilities|LinearColortoVector4 _value_color2_4_bb2f746b4f7bd5defd0542bf99dc459e))
      (Rendering|Material|SetCustomPrimitiveDataFloat (Variables|Default|GetSmokePlane) 8 (/ (Variables|UserOptions|GetOpacity) 1e+02))
      (Rendering|Material|SetCustomPrimitiveDataFloat (Variables|Default|GetSmokePlane) 9 (Variables|UserOptions|GetSmoke%20Tiling))
      (Rendering|Material|SetCustomPrimitiveDataFloat (Variables|Default|GetSmokePlane) 10 (Variables|UserOptions|GetMotion%20Speed))
      (Rendering|Material|SetMaterial (Variables|Default|GetSmokePlane) 0 (CallFunction|GetFogMaterial)))))

(fn ConstructionScript
  (|CallParentFunction)
  (Networking|CanExecuteCosmeticEvents
    (:True
      (CallFunction|SetVisualParameters)
      (Rendering|Material|SetCustomPrimitiveDataFloat (Variables|Default|GetSmokePlane) 11 (Math|Random|RandomFloatinRange 0.0 1e+02)))))
