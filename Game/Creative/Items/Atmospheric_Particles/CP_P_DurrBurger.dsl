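; Creative prop Blueprint for a Durr Burger item that applies an override material to its Sign_Durrr_Top_Spray mesh.
; The construction script sets that material on slot 0 only where cosmetic events can execute.
; GetOverrideMeshMaterial returns the same material for that mesh and nothing for others, also enabling custom depth rendering on the sign as a side effect.
; BeginPlay, Tick and overlap events are disabled stubs.
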(event EventBeginPlay
  (|CallParentFunction) ; This node is disabled and will not be called.\nDrag off pins to build functionality.
)

(event Collision|EventActorBeginOverlap
  (params (OtherActor :type Actor%20Object%20Reference))
  (|CallParentFunction OtherActor) ; This node is disabled and will not be called.\nDrag off pins to build functionality.
)

(event EventTick
  (params (DeltaSeconds :type Float%20%28single-precision%29))
  (|CallParentFunction DeltaSeconds) ; This node is disabled and will not be called.\nDrag off pins to build functionality.
)

(fn GetOverrideMeshMaterial
  (params (MeshComp :type Static%20Mesh%20Component%20Object%20Reference) (MatIdx :type Integer))
  (if (Utilities|Equal(Object) MeshComp (Variables|Default|GetSign_Durrr_Top_Spray))
    (Rendering|SetRenderCustomDepth (Variables|Default|GetSign_Durrr_Top_Spray) true)
    (return (Variables|Default|GetDurrrBurgerMaterialOverride))
    (else
      (return 0))))

(fn ConstructionScript
  (Networking|CanExecuteCosmeticEvents
    (:True
      (bind _durrrburgermaterialoverride (Variables|Default|GetDurrrBurgerMaterialOverride)
        (:then
          (Rendering|Material|SetMaterial (Variables|Default|GetSign_Durrr_Top_Spray) 0 _durrrburgermaterialoverride))
        (:else)))))
